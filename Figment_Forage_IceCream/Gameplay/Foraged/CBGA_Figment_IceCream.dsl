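; Forageable ice cream gameplay actor for Figment that a pawn can consume on the spot or gather.
; Interaction is always offered, and the consume check branches on the attempt.
; The first interaction succeeds only when the pawn's health is not at max and the parent check passes.
; The second follows the GatherEnabled row toggle, playing the gather sound at the actor when enabled, and an invalid pawn is refused.
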(fn BlueprintCanInteract
  (params (InteractingPawn :type Fort%20Pawn%20Object%20Reference) (InteractionBeingAttempted :type EInteractionBeingAttempted%20Enum :default "FirstInteraction") (InteractionType :type TInteractionType%20Enum :default "IT_NoInteraction"))
  (return true))

(fn OnInteractAllowedToConsume
  (params (InteractingPawn :type Fort%20Pawn%20Object%20Reference) (InteractionBeingAttempted :type EInteractionBeingAttempted%20Enum :default "FirstInteraction"))
  (bind _row_gatherenabled (Variables|Gather|GetRow_GatherEnabled))
  (bind _returnvalue (Class|AthenaFunctionLibrary|RowtoBool _row_gatherenabled))
  (Utilities|IsValid InteractingPawn
    (:Is%20Valid
      ;[1 1 1] If second interact allow gather if enabled
      (switch Utilities|FlowControl|Switch|SwitchonEInteractionBeingAttempted InteractionBeingAttempted
        (:FirstInteraction
          (return (and (not (Math|Float|NearlyEqual(Float) (FortPawn|GetHealth InteractingPawn) (FortPawn|GetMaxHealth InteractingPawn))) (|CallParentFunction InteractingPawn "FirstInteraction" "IT_Simple"))))
        (:SecondInteraction
          (if _returnvalue
            (Audio|PlaySoundatLocation (Variables|Default|GetGatherSound) (Transformation|GetActorLocation) "0, 0, 0" 0.4))
          (return _returnvalue))))
    (:Is%20Not%20Valid
      (return false))))

(fn ConstructionScript
  (|CallParentFunction))
